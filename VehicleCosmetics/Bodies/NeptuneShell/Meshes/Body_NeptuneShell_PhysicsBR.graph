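# Physics asset for the NeptuneShell vehicle body mesh, a cosmetic body under VehicleCosmetics.
# It defines a single physics body on the root bone, with no collision shapes listed for it.
# It has no constraints, so no joints link it to other bodies.
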PhysicsAsset /VehicleCosmetics/Bodies/NeptuneShell/Meshes/Body_NeptuneShell_PhysicsBR
bodies {root []}
constraints []
