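# Material function for the Livery02 skin of the NeptuneShell vehicle body, setting material attributes with color, finish and intensity parameters.
# An RGB mask texture blacks out base color in its red and blue areas and doubles as a heightmap for the normal, which drives Fresnel shading.
# Emissive layers panning noise gradients with hue-shifted tints in the RGB areas, over a time-based sine pulse in the alpha channel.
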MaterialFunction /VehicleCosmetics/Bodies/NeptuneShell/Skins/MF_Vehicle_Skin_Neptuneshell_Livery02
graph {nodes [
  Composite() {subgraphName=Vectors}
  n1 = SetMaterialAttributes!(
   MaterialAttributes
   BaseColor=NamedRerouteUsage()
   Metallic=ScalarParameter() {defaultValue=0.5, parameterName=Metallic}
   Roughness=ScalarParameter() {defaultValue=0.1, parameterName=Roughness}
   CustomData0
   CustomData1
   Normal=NamedRerouteUsage()
   ClearCoatBottomNormal
   EmissiveColor=NamedRerouteUsage()
  ) {attributeSetTypes [
    #1
    #2
    #3
    #4
    #5
    #6
    #7
    #8
   ]}
  FunctionOutput(None=@n1)
  NamedRerouteDeclaration(None=Constant3Vector() {constant=<rgba 0.05 0.05 0.05 0>}) {name=TeamColor, nodeColor=<rgba 0.9219 0 1 1>}
  n2 = VectorParameter() {group=Color, parameterName=AccentColor, sortPriority=2, defaultValue=<rgba 0.9739 0 0.0796 1>}
  MaterialExpressionReroute_24 = Reroute(None=@n2.RGB)
  NamedRerouteDeclaration(None=@MaterialExpressionReroute_24.AccentColor) {name=AccentColor, nodeColor=<rgba 0.9219 0 1 1>}
  n3 = VectorParameter() {group=Color, parameterName=PaintColor, sortPriority=3, defaultValue=<rgba 0 0 0 1>}
  MaterialExpressionReroute_25 = Reroute(None=@n3.RGB)
  NamedRerouteDeclaration(None=@MaterialExpressionReroute_25.PaintColor) {name=PaintColor, nodeColor=<rgba 0.9219 0 1 1>}
  n4 = NamedRerouteUsage()
  n5 = Constant3Vector()
  n6 = TextureSampleParameter2D() {parameterName=RGB Mask, texture=&/T_Neptuneshell_Livery02_RGB}
  n7 = Max(A=Max(A=@n6.R, B=@n6.R), B=@n6.B)
  n8 = TextureObject() {texture=&T_Neptuneshell_Livery02_RGB}
  n9 = Constant() {r=-0.5}
  n10 = TextureCoordinate()
  n11 = Constant4Vector() {constant=<rgba 1 1 1 0>}
  n12 = MaterialFunctionCall!(
   Fresnel Power=Constant() {r=3}
   Normal=NamedRerouteDeclaration(None=MaterialFunctionCall!(Height Map=@n8, Normal Map Intensity=@n9, Height Map UV Offset, Coordinates=@n10, Height Map Channel Selector=@n11.RGBA) {materialFunction=/Engine/Functions/Engine_MaterialFunctions03/Procedurals/NormalFromHeightmap}.Result) {name=Normal, nodeColor=<rgba 0.6172 0 1 1>}
  ) {materialFunction=/VehicleCosmetics/Materials/MaterialFunctions/MF_Fresnel_Prime}
  n13 = Saturate(None=Power(Base=OneMinus(None=@n12.Result)) {constExponent=9})
  n14 = Clamp(None=Power(Base=@n12.Result) {constExponent=1}) {minDefault=0.1}
  NamedRerouteDeclaration(None=Multiply(A=LinearInterpolate(A=@n4, B=@n5, Alpha=@n7), B=Max(A=@n13, B=@n14))) {name=BaseColor, nodeColor=<rgba 0 1 0.7656 1>}
  n15 = NamedRerouteUsage()
  n16 = Multiply(A=@n15) {constB=0.2}
  n17 = Sine(None=NamedRerouteUsage()) {period=8}
  n18 = Multiply(A=ConstantBiasScale(None=Abs(None=@n17)) {bias=0.2, scale=0.75}, B=@n6.A)
  n19 = MaterialFunctionCall(Hue Shift Percentage=Constant() {r=-0.01}, Texture=Desaturation(None=Add(A=@n15) {constB=0.25}, Fraction=Constant() {r=-0.5})) {materialFunction=&HueShift}
  n20 = TextureCoordinate() {coordinateIndex=1}
  n21 = LinearInterpolate(
   A=@n16
   B=@n19.Result
   Alpha=MaterialFunctionCall!(Texture Coordinates=@n20, Tiling=Constant() {r=1}, PanSpeed=Constant() {r=4}, Angle, RotateSpeed, Sharpen=Constant() {r=3}, Offset) {materialFunction=/VehicleCosmetics/Materials/MaterialFunctions/MF_Noisemaker_Gradient}.Gradient2
  )
  n22 = Constant() {r=1}
  n23 = Constant() {r=2}
  n24 = Constant() {r=0.5}
  n25 = LinearInterpolate(
   A=LinearInterpolate(A=LinearInterpolate(B=@n16, Alpha=@n18), B=@n21, Alpha=@n6.R)
   B=LinearInterpolate(
    A=@n16
    B=@n19.Result
    Alpha=MaterialFunctionCall(Texture Coordinates=@n20, Tiling=@n22, PanSpeed=@n23, Sharpen=@n23, Offset=@n24) {materialFunction=&MF_Noisemaker_Gradient}.Gradient2
   )
   Alpha=@n6.G
  )
  NamedRerouteDeclaration(None=Multiply(A=LinearInterpolate(A=@n25, B=@n21, Alpha=@n6.B), B=ScalarParameter() {defaultValue=3, group=Color, parameterName=EmissiveIntensity})) {name=Emissive, nodeColor=<rgba 1 0 0.6797 1>}
  NamedRerouteDeclaration(None=Time() {bOverride_Period=true, period=500}) {name=time, nodeColor=<rgba 0.8516 1 0 1>}
  PinBase()
  n26 = VectorParameter() {group=Color, parameterName=BaseColor, sortPriority=1, defaultValue=<rgba 0.0024 0.2113 1 1>}
  n27 = VectorParameter() {group=Color, parameterName=TrimColor, sortPriority=4, defaultValue=<rgba 0.03 0.03 0.03 0>}
  PinBase!(TeamColor=@n26.RGB, AccentColor=@n2.RGB, PaintColor=@n3.RGB, TrimColor=@n27.RGB) {reroutePins [
    TeamColor {expression=&self:MaterialExpressionReroute_23}
    AccentColor {expression=&self:MaterialExpressionReroute_24}
    PaintColor {expression=&self:MaterialExpressionReroute_25}
    TrimColor {expression=&self:MaterialExpressionReroute_43}
   ]}
  MaterialExpressionReroute_23 = Reroute(None=@n26.RGB)
  MaterialExpressionReroute_43 = Reroute(None=@n27.RGB)
 ]}
